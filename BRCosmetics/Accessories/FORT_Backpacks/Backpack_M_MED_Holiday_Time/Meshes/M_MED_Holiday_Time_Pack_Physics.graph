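# Physics asset for the Holiday Time backpack cosmetic, defining secondary motion for its hat, stocking and fur trim bones.
# The hat is a capsule chain from dyn_hat_base through dyn_hat_parent to three segments, with no constraint linking it to dyn_parent.
# Box bodies cover the rest: a two-link stocking toe chain and six separate fur pieces, all rooted on dyn_parent.
# Tight angular limits, especially on twist, restrict the sway.
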PhysicsAsset /BRCosmetics/Accessories/FORT_Backpacks/Backpack_M_MED_Holiday_Time/Meshes/M_MED_Holiday_Time_Pack_Physics
bodies {
 dyn_hat_01 [{shapeType=Capsule, length=2.3606, radius=1.6107, center=<2.2431 0 0>, rotation=<rot 90 0 0>}]
 dyn_hat_02 [{shapeType=Capsule, length=1.6363, radius=1.3195, center=<1.6202 0 0>, rotation=<rot 90 0 0>}]
 dyn_hat_03 [{shapeType=Capsule, length=2.0753, radius=1.5074, center=<2.2107 0 0>, rotation=<rot 90 0 0>}]
 dyn_hat_base [{shapeType=Capsule, length=0.75, radius=0.5}]
 dyn_hat_parent [{shapeType=Capsule, length=3.348, radius=1.8702, center=<2.9125 0 0>, rotation=<rot 90 0 0>}]
 dyn_hatfur_01 [{shapeType=Box, extentX=2, extentY=4.5, extentZ=3, center=<0 0 -1.1779>}]
 dyn_hatfur_02 [{shapeType=Box, extentX=2, extentY=4.5, extentZ=3, center=<0 0 -1.2564>}]
 dyn_hatfur_03 [{shapeType=Box, extentX=2, extentY=4.5, extentZ=3, center=<0 0 -1.2584>}]
 dyn_parent [{shapeType=Box, extentX=2.5, extentY=2.5, extentZ=2.5}]
 dyn_stocking_toe [{shapeType=Box, extentX=8.9446, extentY=7.3027, extentZ=4.8906, center=<4.0398 0 0>}]
 dyn_stocking_tow_01 [{shapeType=Box, extentX=8.9446, extentY=7.3027, extentZ=4.8906, center=<4.0398 0 0>}]
 dyn_stockingfur_01 [{shapeType=Box, extentX=2.01, extentY=6.01, extentZ=4.01, center=<0 0 -1.7006>}]
 dyn_stockingfur_02 [{shapeType=Box, extentX=2.01, extentY=6.01, extentZ=4.01, center=<0 0 -1.7006>}]
 dyn_stockingfur_03 [{shapeType=Box, extentX=2.01, extentY=6.01, extentZ=4.01, center=<0 0 -1.7006>}]
}
constraints [
 {bone1Name=dyn_hat_parent, bone2Name=dyn_hat_base, swing1LimitDegrees=14, swing2LimitDegrees=25, twistLimitDegrees=14}
 {bone1Name=dyn_hat_01, bone2Name=dyn_hat_parent, swing1LimitDegrees=14, swing2LimitDegrees=30, twistLimitDegrees=14}
 {bone1Name=dyn_hat_02, bone2Name=dyn_hat_01, swing1LimitDegrees=14, swing2LimitDegrees=20, twistLimitDegrees=14}
 {bone1Name=dyn_hat_03, bone2Name=dyn_hat_02, swing1LimitDegrees=14, swing2LimitDegrees=14, twistLimitDegrees=4}
 {bone1Name=dyn_stocking_toe, bone2Name=dyn_parent, swing1LimitDegrees=25, swing2LimitDegrees=15, twistLimitDegrees=8}
 {bone1Name=dyn_stockingfur_01, bone2Name=dyn_parent, swing1LimitDegrees=20, swing2LimitDegrees=20, twistLimitDegrees=3}
 {bone1Name=dyn_stockingfur_02, bone2Name=dyn_parent, swing1LimitDegrees=20, swing2LimitDegrees=20, twistLimitDegrees=3}
 {bone1Name=dyn_stockingfur_03, bone2Name=dyn_parent, swing1LimitDegrees=20, swing2LimitDegrees=20, twistLimitDegrees=3}
 {bone1Name=dyn_hatfur_03, bone2Name=dyn_parent, swing1LimitDegrees=20, swing2LimitDegrees=30, twistLimitDegrees=3}
 {bone1Name=dyn_hatfur_02, bone2Name=dyn_parent, swing1LimitDegrees=20, swing2LimitDegrees=30, twistLimitDegrees=3}
 {bone1Name=dyn_hatfur_01, bone2Name=dyn_parent, swing1LimitDegrees=20, swing2LimitDegrees=30, twistLimitDegrees=3}
 {bone1Name=dyn_stocking_tow_01, bone2Name=dyn_stocking_toe, swing1LimitDegrees=8, swing2LimitDegrees=8, twistLimitDegrees=2}
]
