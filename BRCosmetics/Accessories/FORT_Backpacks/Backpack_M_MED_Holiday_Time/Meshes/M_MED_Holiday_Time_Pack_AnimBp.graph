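# Animation Blueprint for the Holiday Time backpack cosmetic, adding procedural offsets and secondary physics motion.
# Spine bend, measured from reference pose, drives an additive root offset feeding two chained RigidBody simulations: fur, damped while targeting, then the main pack physics.
# Jumping applies a delayed upward external force to the second simulation, later reset through a DoOnce.
# Skydive diving or floating then enables additive rotations on three hat bones.
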AnimBlueprint /BRCosmetics/Accessories/FORT_Backpacks/Backpack_M_MED_Holiday_Time/Meshes/M_MED_Holiday_Time_Pack_AnimBp
target_skeleton=&/M_MED_Holiday_Time_Pack_Skeleton
graphs [
 AnimGraph {nodes [
   n1 = "Transform (Modify) Bone - Bone: root"(ComponentPose=Local To Component(LocalPose=Input Pose()), Translation=Get backpack_translation(), Rotation=Get backpack_rotate()) {
    rotationMode=BMM_Additive
    rotationSpace=BCS_BoneSpace
    translationMode=BMM_Additive
    translationSpace=BCS_BoneSpace
    alphaBoolBlend {blendOption=Linear}
    boneToModify {boneName=root}
   }
   n2 = RigidBody(ComponentPose=@n1, Alpha=Get ADS_Alpha() -> (ADS_Alpha:double)) {
    bClampLinearTranslationLimitToRefPose=true
    bOverrideWorldGravity=true
    overridePhysicsAsset=&/M_MED_Holiday_Time_Pack_Fur_Physics
    simulationSpace=BaseBoneSpace
    alphaBoolBlend {blendOption=Linear}
    baseBoneRef {boneName=root}
    componentAppliedLinearAccClamp=<3000 3000 3000>
    componentLinearAccScale=<0.5 0.5 0.5>
    componentLinearVelScale=<0.5 0.5 0.5>
    overrideWorldGravity {z=-750}
   }
   n3 = RigidBody(ComponentPose=@n2, ExternalForce=Get Ext_Force() -> (Ext_Force:Vector)) {
    bClampLinearTranslationLimitToRefPose=true
    bOverrideWorldGravity=true
    overridePhysicsAsset=&/M_MED_Holiday_Time_Pack_Physics
    simulationSpace=BaseBoneSpace
    alphaBoolBlend {blendOption=Linear}
    baseBoneRef {boneName=root}
    componentAppliedLinearAccClamp=<6000 6000 6000>
    componentLinearAccScale=<0.35 0.35 0.35>
    componentLinearVelScale=<0.35 0.35 0.35>
    overrideWorldGravity {z=-900}
   }
   n4 = Get Skydive_Alpha() -> (Skydive_Alpha:double)
   n5 = "Transform (Modify) Bone - Bone: dyn_hat_base"(ComponentPose=@n3, Alpha=@n4, Translation="0,0,0", Rotation="0,20,0", Scale="1,1,1") {
    rotationMode=BMM_Additive
    rotationSpace=BCS_BoneSpace
    translationSpace=BCS_BoneSpace
    alphaBoolBlend {blendOption=Linear}
    boneToModify {boneName=dyn_hat_base}
    rotation {yaw=20}
   }
   n6 = "Transform (Modify) Bone - Bone: dyn_hat_01"(ComponentPose=@n5, Alpha=@n4, Translation="0,0,0", Rotation="0,20,0", Scale="1,1,1") {
    rotationMode=BMM_Additive
    rotationSpace=BCS_BoneSpace
    translationSpace=BCS_BoneSpace
    alphaBoolBlend {blendOption=Linear}
    boneToModify {boneName=dyn_hat_01}
    rotation {yaw=20}
   }
   n7 = "Transform (Modify) Bone - Bone: dyn_hat_03"(ComponentPose=@n6, Alpha=@n4, Translation="0,0,0", Rotation="0,30,0", Scale="1,1,1") {
    rotationMode=BMM_Additive
    rotationSpace=BCS_BoneSpace
    translationSpace=BCS_BoneSpace
    alphaBoolBlend {blendOption=Linear}
    boneToModify {boneName=dyn_hat_03}
    rotation {yaw=30}
   }
   Output Pose(Result=Component To Local(ComponentPose=@n7)) {layerGroup=None}
  ]}
 EventGraph {nodes [
   Event BlueprintUpdateAnimation()
   > n1 = Sequence!(execute) -> (then_0, then_1, then_2:Exec, then_3:Exec)
   n2 = GetOwnerSkeletalMeshComponent()
   n3 = BreakRotator(InRot=BreakTransform(InTransform=GetDeltaTransformFromRefPose(self=@n2, BoneName=spine_05, BaseName=spine_01)).Rotation)
   > Set backpack_rotate(backpack_rotate=MakeRotator(Roll=MapRangeClamped(Value=@n3.Yaw, InRangeA=0, InRangeB=-30, OutRangeB=20)))
   > Set backpack_translation(backpack_translation=MakeVector(Y=MapRangeClamped(Value=@n3.Yaw, InRangeA=0, InRangeB=-30, OutRangeB=-3)))
   n4 = Branch(execute=@n1.then_1, Condition=Get bIsJumping())
   > Delay(Duration=0.25)
   > Set Ext_Force(Ext_Force:Vector="0,0,1250") -> (Output_Get:Vector)
   > n5 = Delay(Duration=0.375)
   Do Once(execute=@n4.else, Reset=@n5.then)
   > Set Ext_Force(Ext_Force="0, 0, 0")
   n6 = Get bIsSkydiveDiveMode()
   n7 = Get bIsSkydiveFloating()
   Set Skydive_Alpha(execute=@n1.then_2, Skydive_Alpha:double=SelectFloat(A=1, B=0, bPickA=OR Boolean(A=@n6, B=@n7))) -> (Output_Get:double)
   Set ADS_Alpha(execute=@n1.then_3, ADS_Alpha:double=SelectFloat(A=0.35, B=1, bPickA=Get bIsTargeting())) -> (Output_Get:double)
   TryGetPawnOwner()
  ]}
]
variables [
 ADS_Alpha {type=number}
 backpack_rotate {type=Rotator}
 backpack_translation {type=Vector}
 Ext_Force {type=Vector}
 Skydive_Alpha {type=number}
]
